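# Material function for RL110 vehicle wheel materials, outputting rim and tire material attributes blended by red vertex color, with a static switch for tireless use.
# Each half feeds masked three-color PBR parameters into the shared wheel setup function, with tread mirroring on tire UVs and normals.
# Effect_Enable adds base-color-tinted emissive to both, plus clamped local-Y vertex offset on the rim, shaped by vertex colors and UV1.
# Time-scrolled Perlin samples provide low- and high-frequency noise signals.
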MaterialFunction /VehicleCosmetics/Wheels/rl110/Materials/MF_Wheel_RL110
graph {
 properties {description=Simplified MF For Wheels}
 nodes [
  Composite() {subgraphName=Color Settings A-C}
  Composite() {subgraphName=Color Settings A-C}
  Composite() {subgraphName=Diffuse Settings w/Tread Mirroring}
  Composite() {subgraphName=Diffuse Settings}
  Composite() {subgraphName=Normal w/Tread Mirroring}
  Composite() {subgraphName=Normal - No Tread Mirroring}
  Composite() {subgraphName=Default Settings}
  Composite() {subgraphName=Default Settings}
  Composite() {subgraphName="Color && Mask"}
  Composite() {subgraphName="Color && Mask w/Tread Mirroring"}
  n1 = ScalarParameter() {
   defaultValue=0.4
   desc=General Roughness that takes effect in absence of masking.
   group=4_Surface Defaults
   parameterName=Rim_RoughnessBase
   sliderMax=1
  }
  n2 = ScalarParameter() {
   defaultValue=1
   desc=General Metalness that takes effect in absence of masking.
   group=4_Surface Defaults
   parameterName=Rim_MetallicBase
   sliderMax=1
  }
  n3 = ScalarParameter() {
   defaultValue=1
   desc="0==False, 1==True: AO gradient affects Emissive values. Use this for Emissives with depth. Values between 0 && 1 are allowed."
   group=5_Special Effects
   parameterName=Rim_AOTintsEmissive
  }
  n4 = ScalarParameter() {
   defaultValue=1
   desc="0==False, 1==True: Blends AO with Diffuse. Values between 0 && 1 are allowed."
   group=5_Special Effects
   parameterName=Rim_AOTintsDiffuse
  }
  n5 = MaterialFunctionCall(R=@n1, G=@n2, B=@n3, A=@n4) {desc="Vector4 container for Rough, Metallic, AO settings", materialFunction=&AppendMany}
  MaterialExpressionReroute_16 = Reroute(None=@n5.RGBA)
  n6 = TextureSampleParameter2D() {group=1_Input Textures, parameterName=Rim_Normal, samplerType=SAMPLERTYPE_Normal, texture=/VehicleCosmetics/Textures/T_BlankNormal_B}
  MaterialExpressionReroute_50 = Reroute(None=@n6.RGB)
  n7 = TextureSampleParameter2D() {desc=Diffuse Input Texture - Optional, group=1_Input Textures, parameterName=Rim_DiffuseTexture, texture=/VehicleCosmetics/Textures/T_White}
  MaterialExpressionReroute_46 = Reroute(None=@n7.RGB)
  n8 = ScalarParameter() {
   desc="0==False, 1==True: Set at 1 if Diffuse overrides color, 0 for no Diffuse. Auto-rounds to 0 or 1."
   group=1_Input Textures
   parameterName=Rim_DiffuseTextureEnable
   sliderMax=1
  }
  MaterialExpressionReroute_47 = Reroute(None=@n8)
  n9 = VectorParameter() {
   desc="Albedo/Diffuse color control in absence of RGB maps, or to fill in the gaps. Leave white if using Diffuse Texture instead."
   group=4_Surface Defaults
   parameterName=Rim_ColorBase
   defaultValue=<rgba 1 1 1 1>
  }
  MaterialExpressionReroute_17 = Reroute(None=@n9.RGB)
  n10 = TextureSampleParameter2D() {
   desc=Mask for Rim - Alpha is AO
   group=1_Input Textures
   parameterName=Rim_MaskRGBA
   samplerType=SAMPLERTYPE_Masks
   texture=/VehicleCosmetics/Textures/T_Black_Masks
   channelNames {a=AO, b=ColorC, g=ColorB, r=ColorA}
  }
  MaterialExpressionReroute_18 = Reroute(None=@n10.RGBA)
  n11 = VectorParameter() {
   desc=Color fill for R value of RGB Mask
   group=2_Rim Properties
   parameterName=Rim_ColorA
   channelNames {a=Unused, b="", g="", r=""}
   defaultValue=<rgba 1 0 0 1>
  }
  MaterialExpressionReroute_1 = Reroute(None=@n11)
  n12 = VectorParameter() {
   desc="Emissive, Roughness, Metallic values for ColorA. Emissive has no limit, Roughness && Metallic are clamped between 0 && 1."
   group=2_Rim Properties
   parameterName=Rim_ColorA_PBR
   channelNames {a=Unused, b=Metallic, g=Roughness, r=Emissive}
   defaultValue=<rgba 0 0.4 1 1>
  }
  MaterialExpressionReroute_2 = Reroute(None=@n12)
  n13 = VectorParameter() {
   desc=Color fill for G value of RGB Mask
   group=2_Rim Properties
   parameterName=Rim_ColorB
   channelNames {a=Unused, b="", g="", r=""}
   defaultValue=<rgba 0 1 0 1>
  }
  MaterialExpressionReroute_3 = Reroute(None=@n13)
  n14 = VectorParameter() {
   desc="Emissive, Roughness, Metallic values for ColorB. Emissive has no limit, Roughness && Metallic are clamped between 0 && 1."
   group=2_Rim Properties
   parameterName=Rim_ColorB_PBR
   channelNames {a=Unused, b=Metallic, g=Roughness, r=Emissive}
   defaultValue=<rgba 0 0.4 1 1>
  }
  MaterialExpressionReroute_6 = Reroute(None=@n14)
  n15 = VectorParameter() {
   desc=Color fill for B value of RGB Mask
   group=2_Rim Properties
   parameterName=Rim_ColorC
   channelNames {a=Unused, b="", g="", r=""}
   defaultValue=<rgba 0 0 1 1>
  }
  MaterialExpressionReroute_7 = Reroute(None=@n15)
  n16 = VectorParameter() {
   desc="Emissive, Roughness, Metallic values for ColorC. Emissive has no limit, Roughness && Metallic are clamped between 0 && 1."
   group=2_Rim Properties
   parameterName=Rim_ColorC_PBR
   channelNames {a=Unused, b=Metallic, g=Roughness, r=Emissive}
   defaultValue=<rgba 0 0.4 1 1>
  }
  MaterialExpressionReroute_8 = Reroute(None=@n16)
  n17 = MaterialFunctionCall!(
   Base_Settings=@MaterialExpressionReroute_16.RoughMetalAOTint
   Normal_Texture=@MaterialExpressionReroute_50.RGB
   Diffuse_Texture=@MaterialExpressionReroute_46.Rim_DiffuseTexture
   DiffuseTextureEnable=@MaterialExpressionReroute_47.Rim_DiffuseTextureEnable
   ColorBase=@MaterialExpressionReroute_17.Rim_ColorBase
   MaskRGBA=@MaterialExpressionReroute_18.Rim_MaskRGBA
   ColorA_Color=@MaterialExpressionReroute_1.ColorA
   ColorA_PBR_Values=@MaterialExpressionReroute_2.ColorA_PBR
   ColorB_Color=@MaterialExpressionReroute_3.ColorB
   ColorB_PBR_Values=@MaterialExpressionReroute_6.ColorB_PBR
   ColorC_Color=@MaterialExpressionReroute_7.ColorC
   ColorC_PBR_Values=@MaterialExpressionReroute_8.ColorC_PBR
  ) {materialFunction=/VehicleCosmetics/Wheels/_SharedMaterials/MF_Wheel_Parent_Simple_InputSetup}
  n18 = GetMaterialAttributes(None=@n17.Result) {attributeGetTypes [#1]}
  n19 = TextureSampleParameter2D() {
   desc=Mask for Rim - Alpha is AO
   group=1_Input Textures
   parameterName=Rim_MaskRGBA
   samplerType=SAMPLERTYPE_Masks
   texture=&T_Black_Masks
   channelNames {a=AO, b=ColorC, g=ColorB, r=ColorA}
  }
  n20 = NamedRerouteUsage()
  n21 = VertexColor()
  n22 = ScalarParameter() {defaultValue=0.52, group=6_Effect, parameterName=None1}
  n23 = Constant()
  n24 = Constant() {r=1}
  n25 = Constant3Vector() {constant=<rgba 0 1 0 0>}
  n26 = LinearInterpolate(
   A=Constant3Vector() {constant=<rgba 0 0 1 0>}
   B=LinearInterpolate(A=@n25, B=Constant3Vector() {constant=<rgba 1 0 0 0>}, Alpha=If(A=@n21.G, B=@n22, "A > B"=@n23, "A == B"=@n24, "A < B"=@n24) {constB=0.52})
   Alpha=If(A=@n21.G, B=ScalarParameter() {defaultValue=0.05, group=6_Effect, parameterName=None2}, "A > B"=@n24, "A == B"=@n23, "A < B"=@n23) {constB=0.05}
  )
  n27 = LinearInterpolate(
   A=LinearInterpolate(A=@n20, B=Multiply(B=@n20) {constA=2}, Alpha=ComponentMask(None=@n26) {r=true})
   B=Multiply(B=@n20) {constA=3}
   Alpha=ComponentMask(None=@n26) {g=true}
  )
  n28 = Multiply(A=OneMinus(None=@n27))
  n29 = Multiply(A=Multiply(A=@n19.RGB, B=Multiply(B=@n28) {constA=0.3}) {constA=1, constB=2}) {constA=1}
  n30 = Add(A=@n29, B=@n29) {constB=-0.3}
  n31 = Multiply(
   A=ScalarParameter() {defaultValue=8, group=6_Effect, parameterName=Rim_EmissiveStreength}
   B=Clamp(None=Power(Base=@n30) {constExponent=2.5})
  ) {constA=5, constB=2}
  n32 = TextureCoordinate() {coordinateIndex=1}
  n33 = Multiply(A=@n27, B=Multiply(A=ComponentMask(None=@n32) {r=true}, B=ComponentMask(None=@n32) {g=true}))
  n34 = Multiply(
   A=Transform(None=@n25) {transformSourceType=TRANSFORMSOURCE_Local}
   B=Multiply(A=Multiply(A=@n21.B, B=@n33), B=ScalarParameter() {defaultValue=2, group=6_Effect, parameterName=wpostrength}) {constB=2}
  ) {constB=2}
  n35 = ScalarParameter() {defaultValue=2, group=6_Effect, parameterName=WPO_Clamp}
  n36 = StaticSwitchParameter(
   True=SetMaterialAttributes!(
    MaterialAttributes=@n17.Result
    EmissiveColor=Multiply(A=@n18.BaseColor, B=@n31) {constA=10, constB=2}
    WorldPositionOffset=Clamp(None=@n34, Max=@n35) {maxDefault=5}
   ) {attributeSetTypes [#2, #3]}
   False=@n17.Result
  ) {defaultValue=true, desc="Eff == True, No Eff == False. True is default.", parameterName=Effect_Enable}
  n37 = VertexColor()
  n38 = ScalarParameter() {
   defaultValue=0.7
   desc=General Roughness that takes effect in absence of masking.
   group=4_Surface Defaults
   parameterName=Tire_RoughnessBase
   sliderMax=1
  }
  n39 = ScalarParameter() {desc=General Metalness that takes effect in absence of masking., group=4_Surface Defaults, parameterName=Tire_MetallicBase, sliderMax=1}
  n40 = ScalarParameter() {
   defaultValue=1
   desc="0==False, 1==True: AO gradient affects Emissive values. Use this for Emissives with depth. Values between 0 && 1 are allowed."
   group=5_Special Effects
   parameterName=Tire_AOTintsEmissive
  }
  n41 = ScalarParameter() {
   defaultValue=1
   desc="0==False, 1==True: Blends AO with Diffuse. Values between 0 && 1 are allowed."
   group=5_Special Effects
   parameterName=Tire_AOTintsDiffuse
  }
  n42 = MaterialFunctionCall(R=@n38, G=@n39, B=@n40, A=@n41) {desc="Vector4 container for Rough, Metallic, AO settings", materialFunction=&AppendMany}
  MaterialExpressionReroute_15 = Reroute(None=@n42.RGBA)
  n43 = TextureSampleParameter2D(UVs=NamedRerouteUsage()) {group=1_Input Textures, parameterName=Tire_Normal, samplerType=SAMPLERTYPE_Normal, texture=&T_BlankNormal_B}
  n44 = MaterialFunctionCall(R=Multiply(A=@n43.R, B=NamedRerouteUsage()) {constB=-1}, G=Multiply(A=@n43.G, B=NamedRerouteUsage()) {constB=-1}, B=@n43.B) {materialFunction=&AppendMany}
  MaterialExpressionReroute_49 = Reroute(None=@n44.RGB)
  n45 = TextureSampleParameter2D(UVs=NamedRerouteUsage()) {desc=Diffuse Input Texture - Optional, group=1_Input Textures, parameterName=Tire_DiffuseTexture, texture=&T_White}
  MaterialExpressionReroute_44 = Reroute(None=@n45.RGB)
  n46 = ScalarParameter() {
   desc="0==False, 1==True: Set at 1 if Diffuse overrides color, 0 for no Diffuse. Auto-rounds to 0 or 1."
   group=1_Input Textures
   parameterName=Tire_DiffuseTextureEnable
   sliderMax=1
  }
  MaterialExpressionReroute_45 = Reroute(None=@n46)
  n47 = VectorParameter() {
   desc="Albedo/Diffuse color control in absence of RGB maps, or to fill in the gaps. Leave white if using Diffuse Texture instead."
   group=4_Surface Defaults
   parameterName=Tire_ColorBase
   defaultValue=<rgba 1 1 1 1>
  }
  MaterialExpressionReroute_19 = Reroute(None=@n47.RGB)
  n48 = TextureSampleParameter2D(UVs=NamedRerouteUsage()) {
   desc=Mask for Tire - Alpha is AO
   group=1_Input Textures
   parameterName=Tire_MaskRGBA
   samplerType=SAMPLERTYPE_Masks
   texture=&T_Black_Masks
   channelNames {a=AO, b=ColorC, g=ColorB, r=ColorA}
  }
  MaterialExpressionReroute_20 = Reroute(None=@n48.RGBA)
  n49 = VectorParameter() {
   desc=Color fill for R value of RGB Mask
   group=3_Tire Properties
   parameterName=Tire_ColorA
   channelNames {a=Unused, b="", g="", r=""}
   defaultValue=<rgba 1 0.1248 0 1>
  }
  MaterialExpressionReroute_9 = Reroute(None=@n49)
  n50 = VectorParameter() {
   desc="Emissive, Roughness, Metallic values for ColorA. Emissive has no limit, Roughness && Metallic are clamped between 0 && 1."
   group=3_Tire Properties
   parameterName=Tire_ColorA_PBR
   channelNames {a=Unused, b=Metallic, g=Roughness, r=Emissive}
   defaultValue=<rgba 0 0.7 0 1>
  }
  MaterialExpressionReroute_10 = Reroute(None=@n50)
  n51 = VectorParameter() {
   desc=Color fill for G value of RGB Mask
   group=3_Tire Properties
   parameterName=Tire_ColorB
   channelNames {a=Unused, b="", g="", r=""}
   defaultValue=<rgba 0 1 0.5029 1>
  }
  MaterialExpressionReroute_11 = Reroute(None=@n51)
  n52 = VectorParameter() {
   desc="Emissive, Roughness, Metallic values for ColorB. Emissive has no limit, Roughness && Metallic are clamped between 0 && 1."
   group=3_Tire Properties
   parameterName=Tire_ColorB_PBR
   channelNames {a=Unused, b=Metallic, g=Roughness, r=Emissive}
   defaultValue=<rgba 0 0.7 0 1>
  }
  MaterialExpressionReroute_12 = Reroute(None=@n52)
  n53 = VectorParameter() {
   desc=Color fill for B value of RGB Mask
   group=3_Tire Properties
   parameterName=Tire_ColorC
   channelNames {a=Unused, b="", g="", r=""}
   defaultValue=<rgba 0.5029 0 1 1>
  }
  MaterialExpressionReroute_13 = Reroute(None=@n53)
  n54 = VectorParameter() {
   desc="Emissive, Roughness, Metallic values for ColorC. Emissive has no limit, Roughness && Metallic are clamped between 0 && 1."
   group=3_Tire Properties
   parameterName=Tire_ColorC_PBR
   channelNames {a=Unused, b=Metallic, g=Roughness, r=Emissive}
   defaultValue=<rgba 0 0.7 0 1>
  }
  MaterialExpressionReroute_14 = Reroute(None=@n54)
  n55 = MaterialFunctionCall(
   Base_Settings=@MaterialExpressionReroute_15.RoughMetalAOTint
   Normal_Texture=@MaterialExpressionReroute_49.Tire_Normal
   Diffuse_Texture=@MaterialExpressionReroute_44.Tire_DiffuseTexture
   DiffuseTextureEnable=@MaterialExpressionReroute_45.Tire_DiffuseTextureEnable
   ColorBase=@MaterialExpressionReroute_19.Tire_ColorBase
   MaskRGBA=@MaterialExpressionReroute_20.Tire_MaskRGBA
   ColorA_Color=@MaterialExpressionReroute_9.ColorA
   ColorA_PBR_Values=@MaterialExpressionReroute_10.ColorA_PBR
   ColorB_Color=@MaterialExpressionReroute_11.ColorB
   ColorB_PBR_Values=@MaterialExpressionReroute_12.ColorB_PBR
   ColorC_Color=@MaterialExpressionReroute_13.ColorC
   ColorC_PBR_Values=@MaterialExpressionReroute_14.ColorC_PBR
  ) {materialFunction=&MF_Wheel_Parent_Simple_InputSetup}
  n56 = ScalarParameter() {defaultValue=8, group=6_Effect, parameterName=Tire_EmissiveStreength}
  n57 = TextureSample() {samplerType=SAMPLERTYPE_LinearColor, texture=/VehicleCosmetics/Wheels/rl110/Textures/T_Tire_RL110_RGB}
  n58 = OneMinus(None=NamedRerouteUsage())
  n59 = Multiply(B=Multiply(B=Add(A=@n58) {constB=-0.3}) {constA=1.5}) {constA=1}
  n60 = Constant() {r=1}
  n61 = TextureSample() {samplerType=SAMPLERTYPE_LinearColor, texture=/VehicleCosmetics/Wheels/rl110/Textures/T_Tire_RL110_RGB1}
  n62 = Multiply(
   A=If(A=Multiply(A=@n57.A, B=@n59) {constA=1}, "A > B"=Constant(), "A == B"=@n60, "A < B"=@n60) {constB=0.05}
   B=Multiply(A=@n61.R, B=@n61.R) {constA=1, constB=2}
  ) {constA=1, constB=2}
  n63 = SetMaterialAttributes!(
   MaterialAttributes=@n55.Result
   EmissiveColor=Multiply(
    A=GetMaterialAttributes(None=@n55.Result) {attributeGetTypes [#1]}.BaseColor
    B=Multiply(A=@n56, B=@n62) {constA=8, constB=2}
   ) {constA=10, constB=2}
  ) {attributeSetTypes [#2]}
  n64 = StaticSwitchParameter(
   True=BlendMaterialAttributes(
    A=@n36
    B=StaticSwitchParameter(True=@n63, False=@n55.Result) {defaultValue=true, desc="Eff == True, No Eff == False. True is default.", parameterName=Effect_Enable}
    Alpha=NamedRerouteDeclaration(None=@n37.R) {name=VertColor_R_Out, nodeColor=<rgba 1 0.5 0 1>}
   )
   False=@n36
  ) {defaultValue=true, desc="Tire == True, No Tire == False. False is default.", group=5_Special Effects, parameterName=Tire_Exists}
  FunctionOutput(None=@n64) {outputName=MaterialAttributes}
  n65 = MaterialFunctionCall() {
   desc="UV - For all Tread Mirrors. Lerp A && B - For Normals"
   materialFunction=/VehicleCosmetics/Wheels/_SharedMaterials/MF_Wheel_TreadMirroring
  }
  NamedRerouteDeclaration(None=@n65.LerpA) {name=Tread_LerpA, nodeColor=<rgba 0 1 0.6953 1>}
  n66 = Multiply(A=Time(), B=ScalarParameter() {defaultValue=1, parameterName=TimeSpeedLOW Freq})
  n67 = Clamp(None=Multiply(
    A=TextureSample(UVs=@n66) {samplerType=SAMPLERTYPE_LinearColor, texture=/Game/Global/Textures/Default/Tileable_Patterns/T_PerlinNoise_Uncompressed}.R
    B=ScalarParameter() {defaultValue=1, parameterName=RandomLOWFreq}
   ))
  NamedRerouteDeclaration(None=Clamp(None=Clamp(None=@n67) {minDefault=0.3}) {maxDefault=1.5}) {name=LOW Freq out Clamp, nodeColor=<rgba 1 0.5156 0 1>}
  NamedRerouteDeclaration(None=@n65.LerpB) {name=Tread_LerpB, nodeColor=<rgba 0 1 0.6953 1>}
  NamedRerouteDeclaration(None=@n65.UV) {name=Tread_UV, nodeColor=<rgba 0 1 0.6953 1>}
  NamedRerouteDeclaration(None=Multiply(B=@n67) {constA=0.2}) {name=LOW Freq out Mult, nodeColor=<rgba 1 0.5156 0 1>}
  n68 = Multiply(A=Time(), B=ScalarParameter() {defaultValue=0.1, parameterName=TimeSpeedHighFreq})
  n69 = TextureSample(UVs=@n68) {samplerType=SAMPLERTYPE_LinearColor, texture=/Game/Weapons/FORT_Melee/Pickaxe_RhombCamo/FX/T_Pickaxe_RhombCamo_Perlin}
  n70 = ScalarParameter() {defaultValue=10, parameterName=RandomHighFreq}
  NamedRerouteDeclaration(None=Clamp(None=Multiply(A=@n69.R, B=@n70))) {name=HIGH Freq out Clamp, nodeColor=<rgba 1 0.5156 0 1>}
  NamedRerouteUsage()
  PinBase()
  PinBase!(ColorA=@n11, ColorA_PBR=@n12, ColorB=@n13, ColorB_PBR=@n14, ColorC=@n15, ColorC_PBR=@n16) {reroutePins [
    ColorA {expression=&self:MaterialExpressionReroute_1}
    ColorA_PBR {expression=&self:MaterialExpressionReroute_2}
    ColorB {expression=&self:MaterialExpressionReroute_3}
    ColorB_PBR {expression=&self:MaterialExpressionReroute_6}
    ColorC {expression=&self:MaterialExpressionReroute_7}
    ColorC_PBR {expression=&self:MaterialExpressionReroute_8}
   ]}
  PinBase!()
  PinBase!(Tire_ColorBase=@n47.RGB, Tire_MaskRGBA=@n48.RGBA) {reroutePins [
    Tire_ColorBase {expression=&self:MaterialExpressionReroute_19}
    Tire_MaskRGBA {expression=&self:MaterialExpressionReroute_20}
   ]}
  PinBase!()
  PinBase()
  PinBase!(Tire_DiffuseTexture=@n45.RGB, Tire_DiffuseTextureEnable=@n46) {reroutePins [
    Tire_DiffuseTexture {expression=&self:MaterialExpressionReroute_44}
    Tire_DiffuseTextureEnable {expression=&self:MaterialExpressionReroute_45}
   ]}
  PinBase!()
  PinBase!(Rim_DiffuseTexture=@n7.RGB, Rim_DiffuseTextureEnable=@n8) {reroutePins [
    Rim_DiffuseTexture {expression=&self:MaterialExpressionReroute_46}
    Rim_DiffuseTextureEnable {expression=&self:MaterialExpressionReroute_47}
   ]}
  PinBase!()
  PinBase!(Tire_Normal=@n44.RGB) {reroutePins [Tire_Normal {expression=&self:MaterialExpressionReroute_49}]}
  PinBase!(ColorA=@n49, ColorA_PBR=@n50, ColorB=@n51, ColorB_PBR=@n52, ColorC=@n53, ColorC_PBR=@n54) {reroutePins [
    ColorA {expression=&self:MaterialExpressionReroute_9}
    ColorA_PBR {expression=&self:MaterialExpressionReroute_10}
    ColorB {expression=&self:MaterialExpressionReroute_11}
    ColorB_PBR {expression=&self:MaterialExpressionReroute_12}
    ColorC {expression=&self:MaterialExpressionReroute_13}
    ColorC_PBR {expression=&self:MaterialExpressionReroute_14}
   ]}
  PinBase!()
  PinBase!(RGB=@n6.RGB) {reroutePins [RGB {expression=&self:MaterialExpressionReroute_50}]}
  PinBase!()
  PinBase!(RoughMetalAOTint=@n42.RGBA) {reroutePins [RoughMetalAOTint {expression=&self:MaterialExpressionReroute_15}]}
  PinBase!()
  PinBase!(RoughMetalAOTint=@n5.RGBA) {reroutePins [RoughMetalAOTint {expression=&self:MaterialExpressionReroute_16}]}
  PinBase!()
  PinBase!(Rim_ColorBase=@n9.RGB, Rim_MaskRGBA=@n10.RGBA) {reroutePins [
    Rim_ColorBase {expression=&self:MaterialExpressionReroute_17}
    Rim_MaskRGBA {expression=&self:MaterialExpressionReroute_18}
   ]}
  Sine(None=@n68)
  Sine(None=@n66)
 ]
}
