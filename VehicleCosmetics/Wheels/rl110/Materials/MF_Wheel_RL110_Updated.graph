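# Material function for the RL110 cosmetic wheel that extends MF_WheelParent_Simple with animated emissive glow on rim and tire, plus vertex displacement.
# Two time-driven sines form a clamped low-frequency pulse, scaled per rim zone by vertex-color G thresholds.
# The pulse drives masked rim glow, texture-thresholded tire glow, and local-space WPO weighted by UV1 and vertex B, disabled at Low quality.
# Vertex-color R blends the rim and tire attribute sets.
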MaterialFunction /VehicleCosmetics/Wheels/rl110/Materials/MF_Wheel_RL110_Updated
graph {nodes [
  n1 = MaterialFunctionCall() {materialFunction=&MF_WheelParent_Simple}
  n2 = GetMaterialAttributes(None=@n1.MaterialAttributes)
  n3 = TextureSampleParameter2D() {
   desc=Mask for Rim - Alpha is AO
   group=1_Input Textures
   parameterName=Rim_MaskRGBA
   samplerSource=SSM_Wrap_WorldGroupSettings
   samplerType=SAMPLERTYPE_Masks
   texture=/VehicleCosmetics/Textures/T_Black_Masks
   channelNames {a=AO, b=ColorC, g=ColorB, r=ColorA}
  }
  n4 = GetMaterialAttributes(None=@n1.MaterialAttributes) {attributeGetTypes [#1]}
  n5 = LinearInterpolate(A=LinearInterpolate(B=@n4.BaseColor, Alpha=@n3.ColorA), B=@n4.BaseColor, Alpha=@n3.ColorB)
  n6 = NamedRerouteUsage()
  n7 = VertexColor()
  n8 = Constant()
  n9 = Constant() {r=1}
  n10 = LinearInterpolate(
   A=Constant3Vector() {constant=<rgba 0 0 1 0>}
   B=LinearInterpolate(
    A=Constant3Vector() {constant=<rgba 0 1 0 0>}
    B=Constant3Vector() {constant=<rgba 1 0 0 0>}
    Alpha=If(A=@n7.G, "A > B"=@n8, "A == B"=@n9, "A < B"=@n9) {constB=0.52}
   )
   Alpha=If(A=@n7.G, "A > B"=@n9, "A == B"=@n8, "A < B"=@n8) {constB=0.05}
  )
  n11 = LinearInterpolate(
   A=LinearInterpolate(A=@n6, B=Multiply(B=@n6) {constA=2}, Alpha=ComponentMask(None=@n10) {r=true})
   B=Multiply(B=@n6) {constA=3}
   Alpha=ComponentMask(None=@n10) {g=true}
  )
  n12 = Multiply(A=OneMinus(None=@n11))
  n13 = Multiply(A=Multiply(A=LinearInterpolate(A=@n5, B=@n4.BaseColor, Alpha=@n3.ColorC), B=Multiply(B=@n12) {constA=0.3}) {constA=1, constB=2}) {constA=1}
  n14 = Add(A=@n13, B=@n13) {constB=-0.3}
  n15 = Multiply(
   A=ScalarParameter() {defaultValue=8, group=6_Effect, parameterName=Rim_EmissiveStreength}
   B=Clamp(None=Power(Base=@n14) {constExponent=2.5})
  ) {constA=5, constB=2}
  n16 = Constant()
  n17 = TextureCoordinate() {coordinateIndex=1}
  n18 = Multiply(A=@n11, B=Multiply(A=ComponentMask(None=@n17) {r=true}, B=ComponentMask(None=@n17) {g=true}))
  n19 = AppendVector(A=@n16, B=Multiply(A=Multiply(A=@n7.B, B=@n18), B=ScalarParameter() {defaultValue=1, group=6_Effect, parameterName=wpostrength}) {constB=2})
  n20 = QualitySwitch(Default=Transform(None=AppendVector(A=@n19, B=@n16)) {transformSourceType=TRANSFORMSOURCE_Local}, Low=Constant())
  n21 = TextureSample() {
   samplerSource=SSM_Wrap_WorldGroupSettings
   samplerType=SAMPLERTYPE_LinearColor
   texture=/VehicleCosmetics/Wheels/rl110/Textures/T_Tire_RL110_RGB
  }
  n22 = Add(A=OneMinus(None=NamedRerouteUsage())) {constB=-0.3}
  n23 = Constant() {r=1}
  n24 = If(A=Multiply(A=@n21.A, B=Multiply(B=@n22) {constA=1.5}) {constA=1}, "A > B"=Constant(), "A == B"=@n23, "A < B"=@n23) {constB=0.05}
  n25 = TextureSample() {
   samplerSource=SSM_Wrap_WorldGroupSettings
   samplerType=SAMPLERTYPE_LinearColor
   texture=/VehicleCosmetics/Wheels/rl110/Textures/T_Tire_RL110_RGB1
  }
  n26 = Multiply(A=@n25.R, B=@n25.R) {constA=1, constB=2}
  n27 = Multiply(
   A=GetMaterialAttributes(None=@n1.MaterialAttributes) {attributeGetTypes [#1]}.BaseColor
   B=Multiply(A=ScalarParameter() {defaultValue=8, group=6_Effect, parameterName=Tire_EmissiveStreength}, B=Multiply(A=@n24, B=@n26) {constA=1, constB=2}) {constA=8, constB=2}
  )
  FunctionOutput(None=BlendMaterialAttributes(
    A=SetMaterialAttributes!(MaterialAttributes=@n2.MaterialAttributes, EmissiveColor=@n15, WorldPositionOffset=@n20) {attributeSetTypes [#2, #3]}
    B=SetMaterialAttributes!(MaterialAttributes=@n2.MaterialAttributes, EmissiveColor=@n27) {attributeSetTypes [#2]}
    Alpha=VertexColor().R
   )) {outputName=MaterialAttributes}
  n28 = Time()
  n29 = Multiply(A=@n28) {constB=1.2}
  n30 = ConstantBiasScale(None=Divide(A=Sine(None=@n29) {period=0.3}))
  n31 = Add(A=@n28) {constB=0.3}
  n32 = ConstantBiasScale(None=Divide(A=Sine(None=@n31) {period=0.4})) {bias=1.2, scale=1}
  NamedRerouteDeclaration(None=Clamp(None=Multiply(A=@n30, B=@n32)) {maxDefault=0.7}) {name=LOW Freq out Clamp, nodeColor=<rgba 1 0.5156 0 1>}
 ]}
