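# Wheel material for the RL110 vehicle wheel cosmetic, set up for skeletal meshes and mesh deformers with default lit shading.
# It uses material attributes and builds entirely from three chained material functions.
# MF_Wheel_RL110_Updated supplies the base surface attributes, which pass through MF_BaseVehicle_Damage for damage and then MF_VehicleDeRez_Transition for the de-rez transition effect.
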Material /VehicleCosmetics/Wheels/rl110/Materials/M_Wheel_RL110
graph {
 properties {
  bUseMaterialAttributes=true
  bUsedWithMeshDeformer=true
  bUsedWithSkeletalMesh=true
  pixelDepthOffsetMode=PDOM_Legacy
  usedShadingModels=Default Lit
 }
 nodes [MaterialFunctionCall(In=MaterialFunctionCall(Material=MaterialFunctionCall() {materialFunction=&/MF_Wheel_RL110_Updated}.MaterialAttributes) {materialFunction=&MF_BaseVehicle_Damage}.Result) {materialFunction=&MF_VehicleDeRez_Transition}]
}
